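# Translucent, unlit, two-sided master material for soft energy trails on Niagara ribbons and mesh particles.
# Emissive blends panning plasma-coil and noise-distorted directional textures, tinted by particle color.
# Opacity combines a smoke-distorted soft trail texture, a sine fade along U and particle alpha.
# World position offset adds SplineThicken width to a sine wobble along WPO_Direction, with Niagara dynamic parameters supplying Time, Offset and Width.
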Material /BRCosmetics/Effects/Niagara/MasterMaterials/SplineThickenTrail/M_SoftTrailEnergy
graph {
 properties {
  bAutomaticallySetUsageInEditor=false
  bUsedWithNiagaraMeshParticles=true
  bUsedWithNiagaraRibbons=true
  blendMode=BLEND_Translucent
  shadingModel=MSM_Unlit
  twoSided=true
  usedShadingModels=Unlit
 }
 nodes [
  n1 = Add(
   A=TextureCoordinate() {uTiling=0.7, vTiling=0.4}
   B=AppendVector(A=DynamicParameter() {paramNames [Time,Offset,Width,Param4]}.Offset, B=Constant())
  )
  n2 = TextureSample(UVs=Panner(Coordinate=Add(A=@n1) {constB=0}, Time=DynamicParameter() {paramNames [Time,Offset,Width,Param4]}.Time) {speedX=0.2}) {samplerType=SAMPLERTYPE_LinearGrayscale, texture=/Game/Effects/Fort_Effects/Emotes/UltraLight/Texture/T_PlasmaCoils}
  n3 = DynamicParameter() {paramNames [Time,Offset,Width,Param4]}
  n4 = Constant()
  n5 = Add(A=Add(A=TextureCoordinate() {uTiling=0.7, vTiling=0.5}, B=AppendVector(A=@n3.Offset, B=@n4)), B=Multiply(A=NamedRerouteUsage()) {constB=0.2})
  n6 = DynamicParameter() {paramNames [Time,Offset,Width,Param4]}
  n7 = LinearInterpolate(Alpha=TextureSample(UVs=Panner(Coordinate=@n5, Time=@n6.Time) {speedX=0.3}) {samplerType=SAMPLERTYPE_LinearGrayscale, texture=/Game/Effects/Fort_Effects/Emotes/UltraLight/Texture/T_DirectionalNoise_Greyscale}.R)
  MaterialFunctionCall(Emissive=Multiply(A=ParticleColor().RGB, B=LinearInterpolate(A=@n2.R, B=@n7) {constAlpha=0.7})) {materialFunction=&MF_FX_EmissiveAdjustments}
  n8 = Add(A=TextureCoordinate(), B=AppendVector(A=DynamicParameter() {paramNames [Time,Offset,Width,Param4]}.Offset, B=Constant()))
  n9 = TextureCoordinate() {uTiling=0.8, vTiling=0.5}
  n10 = AppendVector(A=DynamicParameter() {paramNames [Time,Offset,Width,Param4]}.Offset, B=Constant())
  n11 = TextureSample(UVs=Panner(Coordinate=Add(A=@n9, B=@n10), Time=DynamicParameter() {paramNames [Time,Offset,Width,Param4]}.Time) {speedX=0.1}) {texture=&T_Inky_Smoke_Tile}
  n12 = Multiply(A=NamedRerouteDeclaration(None=ComponentMask(None=@n11.RGB) {g=true, r=true}) {name=Noise, nodeColor=<rgba 1 0.8906 0 1>})
  n13 = TextureSample(UVs=Panner(Coordinate=Add(A=@n8, B=@n12) {constB=0}, Time=DynamicParameter() {paramNames [Time,Offset,Width,Param4]}.Time) {speedX=0.4}) {samplerType=SAMPLERTYPE_Grayscale, texture=&T_Smoke_Trail_Soft}
  n14 = Multiply(A=@n13.R, B=@n13.R)
  n15 = Sine(None=ComponentMask(None=TextureCoordinate()) {r=true}) {period=2}
  Multiply(A=ParticleColor().A, B=Multiply(A=LinearInterpolate(A=@n14, B=@n13.R) {constAlpha=0.2}, B=Power(Base=@n15)))
  n16 = DynamicParameter() {paramNames [Time,Offset,Width,Param4]}
  n17 = Multiply(A=Constant() {r=20}, B=@n16.Width)
  n18 = Multiply(A=Constant() {r=20}, B=@n16.Width)
  n19 = StaticBool()
  n20 = Transform(None=VectorParameter() {parameterName=WPO_Direction, defaultValue=<rgba 0 10 0 0>}.RGB)
  n21 = Add(A=TextureCoordinate(), B=AppendVector(A=DynamicParameter() {paramNames [Time,Offset,Width,Param4]}.Offset, B=Constant()))
  n22 = DynamicParameter() {paramNames [Time,Offset,Width,Param4]}
  n23 = Sine(None=Add(A=ComponentMask(None=@n21) {r=true}, B=Multiply(A=@n22.Time) {constB=0.4})) {period=0.7}
  Reroute(None=Add(
    A=MaterialFunctionCall(WidthBase=@n17, WidthTip=@n18, Expand U or V UV Channel=@n19) {materialFunction=&SplineThicken}.WorldPositionOffset
    B=Multiply(A=@n20, B=@n23)
   ))
  Time()
 ]
}
