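# Translucent, unlit, two-sided master material for glowing trails on Niagara mesh particles.
# SplineThicken widens the mesh between base and tip widths, combined with a camera offset for world position offset.
# Color blends a parameter with particle color, with time-of-day scaling and shared emissive adjustments on RGB.
# Opacity combines view-angle fade, procedural UV glow, tip and optional edge fades, with depth fade skipped on Low and Medium quality.
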Material /BRCosmetics/Effects/Niagara/MasterMaterials/SplineThickenTrail/M_SplineThicken_Glow_Simple
graph {
 properties {
  bAutomaticallySetUsageInEditor=false
  bUsedWithNiagaraMeshParticles=true
  blendMode=BLEND_Translucent
  shadingModel=MSM_Unlit
  twoSided=true
  usedShadingModels=Unlit
 }
 nodes [
  n1 = MaterialFunctionCall(
   WidthBase=ScalarParameter() {defaultValue=2.1, parameterName=WidthBase}
   WidthTip=ScalarParameter() {defaultValue=11, parameterName=WidthTip}
  ) {materialFunction=&SplineThicken}
  Add(
   A=@n1.WorldPositionOffset
   B=MaterialFunctionCall(Offset Amount=ScalarParameter() {parameterName=CameraOffset}) {materialFunction=/Engine/Functions/Engine_MaterialFunctions02/WorldPositionOffset/CameraOffset}.Camera Offset
  )
  n2 = ScalarParameter() {defaultValue=1, group=TimeOfDay, parameterName=TOD_Morning}
  n3 = ScalarParameter() {defaultValue=1, group=TimeOfDay, parameterName=TOD_Day}
  n4 = ScalarParameter() {defaultValue=1, group=TimeOfDay, parameterName=TOD_Evening}
  n5 = ScalarParameter() {defaultValue=1, group=TimeOfDay, parameterName=TOD_Night}
  n6 = VectorParameter() {parameterName=Color, defaultValue=<rgba 1 0.2988 0 1>}
  n7 = ParticleColor()
  n8 = ScalarParameter() {defaultValue=1, parameterName=UseParticleColor}
  MaterialFunctionCall(Emissive=Multiply(
    A=MaterialFunctionCall(Value_Morning=@n2, Value_Day=@n3, Value_Evening=@n4, Value_Night=@n5) {materialFunction=&Scalar_Time_of_Day}.Result
    B=LinearInterpolate(A=@n6.RGB, B=@n7.RGB, Alpha=@n8)
   )) {materialFunction=&MF_FX_EmissiveAdjustments}
  n9 = Power(Base=DotProduct(A=@n1.Normal, B=CameraVectorWS()), Exp=ScalarParameter() {defaultValue=4, parameterName=EdgeFadeContrast}) {constExponent=1}
  n10 = TextureCoordinate()
  n11 = Subtract(A=ComponentMask(None=@n10) {r=true}) {constB=0.5}
  n12 = OneMinus(None=Abs(None=Multiply(A=@n11) {constB=2}))
  n13 = ScalarParameter() {defaultValue=6, parameterName=GlowContrast}
  n14 = OneMinus(None=ComponentMask(None=@n10) {g=true})
  n15 = ScalarParameter() {defaultValue=0.5, parameterName=FadeTipContrast}
  n16 = ScalarParameter() {defaultValue=0.33, parameterName=FadeEdgesMax}
  n17 = Subtract(A=ComponentMask(None=TextureCoordinate()) {g=true}) {constB=0.5}
  n18 = OneMinus(None=Abs(None=Multiply(A=@n17) {constB=2}))
  n19 = Multiply(
   A=Multiply(A=Power(Base=@n12, Exp=@n13) {constExponent=6}, B=Power(Base=@n14, Exp=@n15) {constExponent=1})
   B=LinearInterpolate(B=SmoothStep(Max=@n16, Value=@n18) {constMax=0.33}, Alpha=ScalarParameter() {parameterName=FadeEdges}) {constA=1}
  )
  n20 = DepthFade(Opacity=@n19, FadeDistance=ScalarParameter() {defaultValue=2, parameterName=DepthFadeDistance})
  Multiply(
   A=LinearInterpolate(A=@n6.A, B=@n7.A, Alpha=@n8)
   B=Multiply(A=Saturate(None=@n9), B=QualitySwitch(Default=@n20, Low=@n19, High=@n20, Medium=@n19, Epic=@n20))
  )
 ]
}
